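# Physics asset for the F_MED_Cavalry back bling mesh, defining simulated dangling parts: a spike, a bottle and its liquid.
# Small sphere bodies on the base bones anchor a capsule spike and a capsule bottle through angular-limited constraints.
# The spike swings mostly on one axis, while the bottle swings moderately on both.
# A thin box for the liquid chains off the bottle with tighter limits, letting it tilt within the bottle.
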PhysicsAsset /BRCosmetics/Accessories/FORT_Backpacks/Backpack_F_MED_Cavalry/Meshes/F_MED_Cavalry_Pack_Physics
bodies {
 dyn_bottle_base [{shapeType=Sphere, radius=0.5}]
 dyn_bottle_parent [{shapeType=Capsule, length=6.1126, radius=4.9074, center=<5.1276 0 -0.1385>, rotation=<rot -10.738 0 0>}]
 dyn_liquid [{shapeType=Box, extentX=7.26, extentY=1.01, extentZ=6.51, center=<0.0886 0.0194 0.1065>, rotation=<rot -1.4377 -90 0.0003>}]
 dyn_spike_base [{shapeType=Sphere, radius=0.5}]
 dyn_spike_parent [{shapeType=Capsule, length=24.3021, radius=1.942, center=<2 0 0>, rotation=<rot 0 0 90>}]
}
constraints [
 {bone1Name=dyn_spike_parent, bone2Name=dyn_spike_base, swing1LimitDegrees=1, swing2LimitDegrees=25, twistLimitDegrees=1}
 {bone1Name=dyn_bottle_parent, bone2Name=dyn_bottle_base, swing1LimitDegrees=20, swing2LimitDegrees=20, twistLimitDegrees=3}
 {bone1Name=dyn_liquid, bone2Name=dyn_bottle_parent, swing1LimitDegrees=15, swing2LimitDegrees=5, twistLimitDegrees=2}
]
